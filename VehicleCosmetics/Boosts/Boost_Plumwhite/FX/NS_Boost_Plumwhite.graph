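# Niagara system for the Plumwhite vehicle boost cosmetic, trailing smoke and glitter sprites behind the car.
# SmokeMainParticles and Glitter partition spawns across filtered skeletal mesh sockets, apply painted-variant colors and launch along negative X; a second smoke emitter is disabled.
# Particles follow source movement with influence fading over life, while curves animate size and color.
# The frontend preview flag adds lifetime-ramped drag and shrinks the smoke scale.
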NiagaraSystem /VehicleCosmetics/Boosts/Boost_Plumwhite/FX/NS_Boost_Plumwhite
dependencies {
 usedDataInterfaces [
  &NiagaraDataInterfaceCurve
  &NiagaraDataInterfaceSpriteRendererInfo
  &NiagaraDataInterfaceVectorCurve
  &NiagaraDataInterfaceSkeletalMesh
  &NiagaraDataInterfaceArrayFloat3
 ]
 usedDynamicInputs [&SelectVectorFromArray]
 usedModules [
  &self:SystemSpawnScript.SystemScriptSource.SystemScriptGraph.NiagaraNodeAssignment_0.SetVariables_#1
  &self:SystemSpawnScript.SystemScriptSource.SystemScriptGraph.NiagaraNodeAssignment_1.SetVariables_#2
  &self:SystemSpawnScript.SystemScriptSource.SystemScriptGraph.NiagaraNodeAssignment_2.SetVariables_#3
  &SystemState
 ]
}
emitters {
 Glitter {
  emitterUpdateScript [
   &EmitterState(Life Cycle Mode=::NewEnumerator1, Loop Duration=2, UseLoopDelay=false)
   &VC_FilteredSocketCount()
   &SpawnRate(SpawnRate=40, Use Spawn Probability=false)
  ]
  particleSpawnScript [
   &VC_PaintedVariantsOverride(
    Black=<rgba 0.01 0.01 0.01 1>
    BurntSienna=<rgba 0.0522 0.0161 0.0064 1>
    Cobalt=<rgba 0.0687 0.1028 0.2135 1>
    Crimson=<rgba 0.113 0 0.0031 1>
    ForestGreen=<rgba 0 0.09 0.0045 1>
    Gold:LinearColor=<rgba 1 0.25 0.1 1>
    Grey=<rgba 0.0522 0.0522 0.0522 1>
    Lime=<rgba 0.0875 0.2087 0 1>
    Pink=<rgba 0.1565 0.0845 0.1687 1>
    Purple=<rgba 0.0859 0 0.1094 1>
    Saffron=<rgba 0.25 0.1917 0 1>
    SkyBlue=<rgba 0 0.1109 0.1667 1>
    TitaniumWhite=<rgba 0.09 0.09 0.09 1>
   )
   &InitializeParticle(
    Color Mode=::NewEnumerator0
    Lifetime Max=0.8
    Lifetime Min=0.5
    Lifetime Mode=::NewEnumerator1
    Sprite Rotation Mode=::NewEnumerator2
    Sprite Size Mode=::NewEnumerator3
    Uniform Sprite Size=&RandomRangeFloat(Minimum=13, Maximum=22)
    UsePositionOffset=false
    Write Material Random=true
   )
   &PartitionParticles(Number Of Partitions=$Emitter.PortCount)
   &SkeletalMeshLocation(
    Bone / Socket Index=$Particles.PartitionParticles.Partition
    Bone / Socket Sampling Mode=::NewEnumerator5
    MaskSpawnGroup=false
    Mesh Sampling Type=::NewEnumerator6
    Sampled Position Offset=&AddVectorToPosition(Position=<0 0 0>, Vector=$User.DTCarOffset)
    Skeletal Mesh=$System.SkeletalMesh
   )
   &AddVelocity(Velocity=&RandomRangeVector(Minimum=<-720 0 -15>, Maximum=<-720 0 75>))
  ]
  particleUpdateScript [
   &ParticleState()
   &ScaleSpriteSize(
    Uniform Curve Scale=3.5
    Uniform Curve Sprite Scale=DataInterface() {propertyValues {Curve {keys [
        {arriveTangent=1.1216, interpMode=RCIM_Cubic, leaveTangent=1.1216, tangentMode=RCTM_User, time=0, value=1.5}
        {arriveTangent=4.7775, interpMode=RCIM_Cubic, leaveTangent=4.7775, tangentMode=RCTM_User, time=0.7, value=5}
        {arriveTangent=-1.7854, interpMode=RCIM_Cubic, leaveTangent=-1.7854, tangentMode=RCTM_User, time=1, value=4.5}
       ]}}}
   )
   &DynamicMaterialParameters(
    Index 0 Param 1=&FloatFromCurve(FloatCurve=DataInterface() {propertyValues {Curve {keys [<at 0 0>, {arriveTangent=2.5, interpMode=RCIM_Cubic, leaveTangent=2.5, time=0.15, value=0}, <at 0.4 1 RCIM_Cubic>]}}})
    Index 0 Param 2=&FloatFromCurve(FloatCurve=DataInterface() {propertyValues {Curve {keys [<at 0 0.85 RCIM_Cubic>,<at 1 1 RCIM_Cubic>]}}})
    Index 0 Param 3=&RandomRangeFloat(Minimum=0.1, Maximum=0.225)
    Index 0 Param 4=&FloatFromCurve(FloatCurve=DataInterface() {propertyValues {Curve {keys [<at 0 20 RCIM_Cubic>,<at 1 5 RCIM_Cubic>]}}})
    Module.Index0Param0Write=true
    Module.Index0Param1Write=true
    Module.Index0Param2Write=true
    Module.Index0Param3Write=true
   ) {enabled=false}
   &Drag(
    Drag=&MakeCustomFloatFromBool(
     Bool=$User.bIsFrontendPreview
     True Float=&FloatFromCurve(FloatCurve=DataInterface() {propertyValues {Curve {keys [<at 0 0>,<at 0.9 1>]}}})
     False Float=0
    )
    Use Linear Drag=true
    Use Rotational Drag=true
   )
   &InheritSourceMovement(
    Applied Position Scale=&VectorFromFloat(Value=&FloatFromCurve(FloatCurve=DataInterface() {propertyValues {Curve {keys [
          {interpMode=RCIM_Cubic, tangentMode=RCTM_Break, time=0, value=1}
          {arriveTangent=-2.9025, interpMode=RCIM_Cubic, tangentMode=RCTM_Break, time=1, value=0}
         ]}}}))
    ApplyForce=false
    ApplyPosition=true
    ApplyPositionFalloff=false
   )
   &SolveForcesAndVelocity(Clamp Velocity=false, Limit Acceleration=false)
   &SubUVAnimation(
    Direct Frame Index=12
    "Sprite Renderer "=DataInterface() {propertyValues {SpriteRenderer=&self:Glitter_0.NiagaraSpriteRendererProperties_0}}
    SubUV Animation Mode:ENiagara_SubUVLookupModeV2=::NewEnumerator5
   )
   &ScaleColor(
    Scale Alpha=&FloatFromCurve(FloatCurve=DataInterface() {propertyValues {Curve {keys [<at 0 1>,<at 0.75 1>,<at 1 0>]}}})
    Scale RGB=&VectorFromFloat(Value=&FloatFromCurve(FloatCurve=DataInterface() {propertyValues {Curve {
         defaultValue=0
         keys [
          {interpMode=RCIM_Cubic, tangentMode=RCTM_User, time=0, value=0}
          {interpMode=RCIM_Cubic, tangentMode=RCTM_User, time=0.5, value=1}
          {tangentMode=RCTM_User, time=1, value=1}
         ]
        }}}))
    ScaleA=true
    ScaleRGB=true
   )
  ]
  rendererClasses [&NiagaraSpriteRendererProperties]
  renderers [{rendererClass=&NiagaraSpriteRendererProperties}]
 }
 SmokeMainParticles {
  emitterUpdateScript [
   &EmitterState(Life Cycle Mode=::NewEnumerator1, Loop Duration=2, UseLoopDelay=false)
   &VC_FilteredSocketCount()
   &SpawnRate(SpawnRate=40, Use Spawn Probability=false)
  ]
  particleSpawnScript [
   &VC_PaintedVariantsOverride(
    Black=<rgba 0.01 0.01 0.01 1>
    BurntSienna=<rgba 0.0522 0.0161 0.0064 1>
    Cobalt=<rgba 0.0687 0.1028 0.2135 1>
    Crimson=<rgba 0.113 0 0.0031 1>
    DefaultColor=<rgba 0.1094 0.1094 0.1094 1>
    ForestGreen=<rgba 0 0.09 0.0045 1>
    Gold:LinearColor=<rgba 1 0.25 0.1 1>
    Grey=<rgba 0.0522 0.0522 0.0522 1>
    Lime=<rgba 0.0875 0.2087 0 1>
    Pink=<rgba 0.1565 0.0845 0.1687 1>
    Purple=<rgba 0.0859 0 0.1094 1>
    Saffron=<rgba 0.25 0.1917 0 1>
    SkyBlue=<rgba 0 0.1109 0.1667 1>
    TitaniumWhite=<rgba 0.09 0.09 0.09 1>
   )
   &InitializeParticle(
    Color Mode=::NewEnumerator0
    Lifetime=0.8
    Sprite Rotation Angle=&RandomRangeFloat(Minimum=-0.15, Maximum=0.15)
    Sprite Rotation Mode=::NewEnumerator1
    Sprite Size Mode=::NewEnumerator3
    Uniform Sprite Size=&RandomRangeFloat(Minimum=8, Maximum=10)
    UsePositionOffset=false
    Write Material Random=true
   )
   &PartitionParticles(Number Of Partitions=$Emitter.PortCount)
   &SkeletalMeshLocation(
    Bone / Socket Index=$Particles.PartitionParticles.Partition
    Bone / Socket Sampling Mode=::NewEnumerator5
    MaskSpawnGroup=false
    Mesh Sampling Type=::NewEnumerator6
    Sampled Position Offset=&AddVectorToPosition(Position=<0 0 0>, Vector=$User.DTCarOffset)
    Skeletal Mesh=$System.SkeletalMesh
   )
   &AddVelocity(Velocity=&RandomRangeVector(Minimum=<-720 0 -15>, Maximum=<-720 0 70>))
  ]
  particleUpdateScript [
   &ParticleState()
   &ScaleSpriteSize(
    Uniform Curve Scale=&MakeCustomFloatFromBool(Bool=$User.bIsFrontendPreview, True Float=2, False Float=3)
    Uniform Curve Sprite Scale=DataInterface() {propertyValues {Curve {keys [
        {arriveTangent=-11.0869, interpMode=RCIM_Cubic, leaveTangent=-11.0869, tangentMode=RCTM_User, time=0, value=10.75}
        {arriveTangent=1.3439, interpMode=RCIM_Cubic, leaveTangent=1.3439, tangentMode=RCTM_User, time=0.3, value=9}
        {arriveTangent=-1.3093, interpMode=RCIM_Cubic, leaveTangent=-1.3093, tangentMode=RCTM_User, time=1, value=5.5}
       ]}}}
   )
   &DynamicMaterialParameters(
    Index 0 Param 1=&FloatFromCurve(FloatCurve=DataInterface() {propertyValues {Curve {keys [<at 0 0>, {arriveTangent=2.5, interpMode=RCIM_Cubic, leaveTangent=2.5, time=0.15, value=0}, <at 0.4 1 RCIM_Cubic>]}}})
    Index 0 Param 2=&FloatFromCurve(FloatCurve=DataInterface() {propertyValues {Curve {keys [<at 0 0.85 RCIM_Cubic>,<at 1 1 RCIM_Cubic>]}}})
    Index 0 Param 3=&RandomRangeFloat(Minimum=0.1, Maximum=0.225)
    Index 0 Param 4=&FloatFromCurve(FloatCurve=DataInterface() {propertyValues {Curve {keys [<at 0 20 RCIM_Cubic>,<at 1 5 RCIM_Cubic>]}}})
    Module.Index0Param0Write=true
    Module.Index0Param1Write=true
    Module.Index0Param2Write=true
    Module.Index0Param3Write=true
   ) {enabled=false}
   &Drag(
    Drag=&MakeCustomFloatFromBool(
     Bool=$User.bIsFrontendPreview
     True Float=&FloatFromCurve(FloatCurve=DataInterface() {propertyValues {Curve {keys [<at 0 0>,<at 0.9 1>]}}})
     False Float=0
    )
    Use Linear Drag=true
    Use Rotational Drag=true
   )
   &InheritSourceMovement(
    Applied Position Scale=&VectorFromFloat(Value=&FloatFromCurve(FloatCurve=DataInterface() {propertyValues {Curve {keys [
          {interpMode=RCIM_Cubic, tangentMode=RCTM_Break, time=0, value=1}
          {arriveTangent=-2.9025, interpMode=RCIM_Cubic, tangentMode=RCTM_Break, time=1, value=0}
         ]}}}))
    ApplyForce=false
    ApplyPosition=true
    ApplyPositionFalloff=false
   )
   &SolveForcesAndVelocity(Clamp Velocity=false, Limit Acceleration=false)
   &SubUVAnimation(
    "Sprite Renderer "=DataInterface() {propertyValues {SpriteRenderer=&self:SmokeMainParticles_0.NiagaraSpriteRendererProperties_0}}
    UseEndFrame=false
    UseStartFrame=false
   )
   &ScaleColor(
    Scale RGB=&VectorFromCurve(VectorCurve=DataInterface() {propertyValues {XCurve {keys [<at 0.3 0.95>,<at 0.55 0.2>]}, YCurve {keys [<at 0.3 0.95>,<at 0.55 0.2>]}, ZCurve {keys [<at 0.3 0.95>,<at 0.55 0.2>]}}})
    ScaleA=true
    ScaleRGB=true
   )
  ]
  rendererClasses [&NiagaraSpriteRendererProperties]
  renderers [{rendererClass=&NiagaraSpriteRendererProperties}]
 }
 SmokeMainParticles002 {
  bEnabled=false
  emitterUpdateScript [
   &EmitterState(Life Cycle Mode=::NewEnumerator1, Loop Duration=2, UseLoopDelay=false)
   &VC_FilteredSocketCount()
   &SpawnRate(SpawnRate=40, Use Spawn Probability=false)
  ]
  particleSpawnScript [
   &VC_PaintedVariantsOverride(
    Black=<rgba 0.01 0.01 0.01 1>
    BurntSienna=<rgba 0.0522 0.0161 0.0064 1>
    Cobalt=<rgba 0.0687 0.1028 0.2135 1>
    Crimson=<rgba 0.113 0 0.0031 1>
    ForestGreen=<rgba 0 0.09 0.0045 1>
    Gold:LinearColor=<rgba 1 0.25 0.1 1>
    Grey=<rgba 0.0522 0.0522 0.0522 1>
    Lime=<rgba 0.0875 0.2087 0 1>
    Pink=<rgba 0.1565 0.0845 0.1687 1>
    Purple=<rgba 0.0859 0 0.1094 1>
    Saffron=<rgba 0.25 0.1917 0 1>
    SkyBlue=<rgba 0 0.1109 0.1667 1>
    TitaniumWhite=<rgba 0.09 0.09 0.09 1>
   )
   &InitializeParticle(
    Color Mode=::NewEnumerator0
    Lifetime=0.6
    Sprite Rotation Angle=&RandomRangeFloat(Minimum=-0.15, Maximum=0.15)
    Sprite Rotation Mode=::NewEnumerator1
    Sprite Size Mode=::NewEnumerator3
    Uniform Sprite Size=&RandomRangeFloat(Minimum=8, Maximum=10)
    UsePositionOffset=false
    Write Material Random=true
   )
   &PartitionParticles(Number Of Partitions=$Emitter.PortCount)
   &SkeletalMeshLocation(
    Bone / Socket Index=$Particles.PartitionParticles.Partition
    Bone / Socket Sampling Mode=::NewEnumerator5
    MaskSpawnGroup=false
    Mesh Sampling Type=::NewEnumerator6
    Sampled Position Offset=&AddVectorToPosition(Position=<0 0 0>, Vector=$User.DTCarOffset)
    Skeletal Mesh=$System.SkeletalMesh
   )
   &AddVelocity(Velocity=&RandomRangeVector(Minimum=<-720 0 150>, Maximum=<-720 0 0>))
  ]
  particleUpdateScript [
   &ParticleState()
   &ScaleSpriteSize(
    Uniform Curve Scale=3.5
    Uniform Curve Sprite Scale=DataInterface() {propertyValues {Curve {keys [
        {arriveTangent=1.1216, interpMode=RCIM_Cubic, leaveTangent=1.1216, tangentMode=RCTM_User, time=0, value=1.5}
        {arriveTangent=4.7775, interpMode=RCIM_Cubic, leaveTangent=4.7775, tangentMode=RCTM_User, time=0.7, value=5}
        {arriveTangent=-1.7854, interpMode=RCIM_Cubic, leaveTangent=-1.7854, tangentMode=RCTM_User, time=1, value=8}
       ]}}}
   )
   &DynamicMaterialParameters(
    Index 0 Param 1=&FloatFromCurve(FloatCurve=DataInterface() {propertyValues {Curve {keys [<at 0 0>, {arriveTangent=2.5, interpMode=RCIM_Cubic, leaveTangent=2.5, time=0.15, value=0}, <at 0.4 1 RCIM_Cubic>]}}})
    Index 0 Param 2=&FloatFromCurve(FloatCurve=DataInterface() {propertyValues {Curve {keys [<at 0 0.85 RCIM_Cubic>,<at 1 1 RCIM_Cubic>]}}})
    Index 0 Param 3=&RandomRangeFloat(Minimum=0.1, Maximum=0.225)
    Index 0 Param 4=&FloatFromCurve(FloatCurve=DataInterface() {propertyValues {Curve {keys [<at 0 20 RCIM_Cubic>,<at 1 5 RCIM_Cubic>]}}})
    Module.Index0Param0Write=true
    Module.Index0Param1Write=true
    Module.Index0Param2Write=true
    Module.Index0Param3Write=true
   ) {enabled=false}
   &Drag(
    Drag=&MakeCustomFloatFromBool(
     Bool=$User.bIsFrontendPreview
     True Float=&FloatFromCurve(FloatCurve=DataInterface() {propertyValues {Curve {keys [<at 0 0>,<at 0.9 1>]}}})
     False Float=0
    )
    Use Linear Drag=true
    Use Rotational Drag=true
   )
   &InheritSourceMovement(
    Applied Position Scale=&VectorFromFloat(Value=&FloatFromCurve(FloatCurve=DataInterface() {propertyValues {Curve {keys [
          {interpMode=RCIM_Cubic, tangentMode=RCTM_Break, time=0, value=1}
          {arriveTangent=-2.9025, interpMode=RCIM_Cubic, tangentMode=RCTM_Break, time=1, value=0}
         ]}}}))
    ApplyForce=false
    ApplyPosition=true
    ApplyPositionFalloff=false
   )
   &SolveForcesAndVelocity(Clamp Velocity=false, Limit Acceleration=false)
   &SubUVAnimation(
    Direct Frame Index=12
    "Sprite Renderer "=DataInterface() {propertyValues {SpriteRenderer=&self:SmokeMainParticles002_0.NiagaraSpriteRendererProperties_0}}
    SubUV Animation Mode:ENiagara_SubUVLookupModeV2=::NewEnumerator5
   ) {enabled=false}
   &ScaleColor(
    Scale RGB=&VectorFromFloat(Value=&FloatFromCurve(FloatCurve=DataInterface() {propertyValues {Curve {keys [<at 0 0>,<at 0.55 0.2>]}}}))
    ScaleA=true
    ScaleRGB=true
   )
  ]
  rendererClasses [&NiagaraSpriteRendererProperties]
  renderers [{rendererClass=&NiagaraSpriteRendererProperties}]
 }
}
summary {
 systemName=NS_Boost_Plumwhite
 emitters [
  {emitterName=SmokeMainParticles, rendererClasses [&NiagaraSpriteRendererProperties]}
  {emitterName=Glitter, rendererClasses [&NiagaraSpriteRendererProperties]}
  {bEnabled=false, emitterName=SmokeMainParticles002, rendererClasses [&NiagaraSpriteRendererProperties]}
 ]
 userVariables [
  User.bIsFrontendPreview {type=NiagaraBool}
  User.bIsPainted {type=NiagaraBool}
  User.blsIsCoroColor {type=NiagaraBool}
  User.bPreviewVehicleSpeed {type=NiagaraBool}
  User.ColorVariantID {type=NiagaraInt32}
  User.DTCarOffset {type=Vector3f, description="", defaultValue {struct=Vector3f, value=<0 0 0>}}
  User.GameModeID {type=NiagaraInt32}
  User.PaintedColor {type=LinearColor, description="", defaultValue {struct=LinearColor, value=<rgba 1 1 1 1>}}
  User.PaintedScalar {type=NiagaraFloat, defaultValue=1, description=""}
  User.PortCount {type=NiagaraInt32, defaultValue=1}
 ]
}
